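# Translucent, unlit, two-sided Niagara material for the StudyBench emote's toon effects, supporting ribbons, sprites and meshes.
# A panning tiling texture, jittered by time-stepped noise distortion and dynamic-parameter offsets, feeds erosion with a narrow divide-and-saturate edge for toon-like shapes.
# Particle color drives exposure-adjusted emissive, while opacity math adds depth and near-camera fades.
# Dynamic parameters also supply erode, reveal and WPO values, alongside a plasma-noise vertex-normal offset.
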Material /BRCosmetics/Animation/Game/MainPlayer/Emotes/StudyBench/FX/M_StudyBench_Toon_New
graph {
 properties {
  bAllowNegativeEmissiveColor=true
  bUsedWithNiagaraMeshParticles=true
  bUsedWithNiagaraRibbons=true
  bUsedWithNiagaraSprites=true
  blendMode=BLEND_Translucent
  pixelDepthOffsetMode=PDOM_Legacy
  shadingModel=MSM_Unlit
  twoSided=true
  usedShadingModels=Unlit
 }
 nodes [
  Constant()
  n1 = ParticleColor()
  n2 = TextureCoordinate() {coordinateIndex=1}
  n3 = AppendVector(A=Constant() {r=1}, B=ScalarParameter() {defaultValue=1, parameterName=Texture X Size})
  n4 = Multiply(A=@n2, B=@n3)
  n5 = Panner(
   Coordinate=StaticSwitchParameter(
    True=MaterialFunctionCall(XY=@n4) {materialFunction=&Swizzle}.YX
    False=Multiply(A=@n2, B=@n3)
   ) {defaultValue=true, parameterName=UseSwizzle}
   Speed=AppendVector(A=ScalarParameter() {parameterName=X Speed}, B=ScalarParameter() {parameterName=Y Speed})
  ) {bFractionalPart=true}
  n6 = NamedRerouteUsage()
  n7 = MaterialFunctionCall(
   X=Add(
    A=ScalarParameter() {parameterName=Texture X Offset}
    B=DynamicParameter() {defaultValue=<rgba 1 0 -1 1>, paramNames [Erode, Random Offset, Reveal, WPO]}.Random Offset
   )
   Y=DynamicParameter() {parameterIndex=1, defaultValue=<rgba 1 0 -1 1>, paramNames [Random Offset Y, "", "", ""]}.Random Offset Y
  ) {materialFunction=&MakeFloat2}
  n8 = MaterialFunctionCall(XY=@n7.Result) {materialFunction=&Swizzle}
  n9 = TextureSampleParameter2D(UVs=Add(A=Add(A=@n5, B=@n6), B=StaticSwitchParameter(True=@n7.Result, False=@n8.YX) {defaultValue=true, parameterName=UseSwizzle})) {
   automaticViewMipBias=false
   parameterName=TilingTexture
   texture=/BRCosmetics/Weapons/FORT_Melee/Pickaxe_EmeraldGlassGreen/FX/T_Pickaxe_EmeraldGlassGreen_Lightning
  }
  n10 = ScalarParameter() {defaultValue=1, group=Erosion, parameterName=Erosion Softness}
  n11 = MaterialFunctionCall!(
   Gradient=MaterialFunctionCall(Gradient=@n9.B, Erosion Amount=NamedRerouteUsage(), Softness=@n10) {materialFunction=/BRCosmetics/Athena/Effects/Particles/Skydiving/NS_Contrail_DarkAndLight/Functions/MF_Erosion}.Default Range
   Erosion Amount=NamedRerouteUsage()
   Softness=@n10
  ) {materialFunction=&MF_Erosion}
  n12 = Saturate(None=Divide(A=@n11.Default Range) {constB=0.1})
  MaterialExpressionReroute_6 = Reroute(None=LinearInterpolate(A=@n1.RGB, B=@n1.RGB, Alpha=Multiply(A=@n9.R, B=@n12)))
  MaterialExpressionScalarParameter_24 = ScalarParameter() {group=Options, parameterName=Use Eye Adaption Exposure Control for Emissives}
  MaterialExpressionEyeAdaptation_1 = EyeAdaptation()
  n13 = Custom!(
   LightValue=@MaterialExpressionReroute_6
   Alpha=@MaterialExpressionScalarParameter_24
   Adaptation=@MaterialExpressionEyeAdaptation_1.EyeAdaptation
  ) {
   code="float LerpLogScale = -Alpha * log(Adaptation);\r\nfloat Scale = exp(LerpLogScale);\r\nreturn LightValue * Scale;\r\n\r\n\r\n"
   inputs [
    {
     inputName=LightValue
     input {expression=&self:MaterialExpressionReroute_6}
    }
    {
     inputName=Alpha
     input {expression=&self:MaterialExpressionScalarParameter_24}
    }
    {
     inputName=Adaptation
     input {expression=&self:MaterialExpressionEyeAdaptation_1}
    }
   ]
  }
  MaterialFunctionCall(Emissive=@n13) {materialFunction=&MF_FX_EmissiveAdjustments}
  n14 = ComponentMask(None=TextureCoordinate()) {g=true}
  n15 = Saturate(None=Multiply(A=Power(Base=@n14) {constExponent=1}, B=ScalarParameter() {defaultValue=6, parameterName=Leading edge fade intensity}) {constB=5})
  n16 = VectorParameter() {defaultValue=<rgba -0.1 0 0 1>}
  n17 = MaterialFunctionCall(Float2=Add(
    A=TextureCoordinate()
    B=MaterialFunctionCall(X=@n16.R) {materialFunction=&MakeFloat2}.Result
   )) {materialFunction=&BreakOutFloat2Components}
  n18 = TextureSample(UVs=MaterialFunctionCall(X=OneMinus(None=@n17.G), Y=@n17.R) {materialFunction=&MakeFloat2}.Result) {
   samplerSource=SSM_Clamp_WorldGroupSettings
   texture=/BRCosmetics/Athena/Effects/Particles/Skydiving/NS_Hightower_Date/Hightower_Ribbon_LeadingEdge_Mask
  }
  Multiply(A=@n15, B=@n18.R)
  n19 = Power(Base=OneMinus(None=@n14)) {constExponent=1}
  n20 = ScalarParameter() {defaultValue=6, parameterName=Trailing edge fade intensity}
  Multiply(A=@n15, B=Saturate(None=Multiply(A=@n19, B=@n20) {constB=5}))
  n21 = Multiply(A=NamedRerouteUsage())
  n22 = ShadingPathSwitch(
   Default=DepthFade(Opacity=@n21, FadeDistance=ScalarParameter() {defaultValue=5, group=Depth Fade, parameterName=NearFadeDistance})
   Mobile=@n21
  ) {desc=Migrated from FeatureLevelSwitch}
  n23 = ScalarParameter() {group=Depth Fade, parameterName=Use Depth Fade}
  n24 = MaterialFunctionCall(Distance=ScalarParameter() {defaultValue=128, parameterName=CameraFadeDistance}) {materialFunction=&Near_Camera_Fade}
  n25 = Constant() {r=1}
  Multiply(
   A=Multiply(
    A=LinearInterpolate(A=@n21, B=@n22, Alpha=@n23)
    B=ShadingPathSwitch(Default=@n24.Result, Mobile=@n25) {desc=Migrated from FeatureLevelSwitch}
   )
   B=NamedRerouteUsage()
  )
  n26 = ComponentMask(None=TextureCoordinate()) {g=true}
  n27 = ScalarParameter() {defaultValue=0.6}
  n28 = Multiply(
   A=TextureSample(UVs=Panner() {speedX=0.5}) {texture=&/T_Plasma_Noise}.R
   B=OneMinus(None=Add(A=@n26, B=@n27) {constB=0.5})
  )
  Multiply(A=Multiply(A=Saturate(None=@n28), B=VertexNormalWS()), B=NamedRerouteUsage())
  n29 = DynamicParameter() {defaultValue=<rgba 1 0 -1 1>, paramNames [Erode, Random Offset, Reveal, WPO]}
  NamedRerouteDeclaration(None=OneMinus(None=Multiply(A=@n29.Erode, B=@n15))) {name=Edge Masks, nodeColor=<rgba 0 1 0.8125 1>}
  n30 = TextureCoordinate() {coordinateIndex=1, uTiling=0.46, vTiling=0.2476}
  n31 = Time()
  n32 = ScalarParameter() {defaultValue=16, parameterName=UVDistortSpeed}
  n33 = Divide(A=Ceil(None=Multiply(A=@n31, B=@n32) {constB=16})) {constB=4}
  n34 = ComponentMask(None=TextureSample(UVs=Panner(Coordinate=@n30, Time=@n33) {speedX=0.4, speedY=0.4}) {samplerType=SAMPLERTYPE_LinearColor, texture=/Game/Global/Textures/Default/Noise/TilingNoise06_N_uncompressed}.RGB) {g=true, r=true}
  n35 = LinearInterpolate(B=Multiply(A=Subtract(A=@n34) {constB=0.5}) {constB=0.1}, Alpha=ScalarParameter() {defaultValue=1, parameterName=Use Distortion})
  NamedRerouteDeclaration(None=StaticSwitchParameter(True=MaterialFunctionCall(XY=@n35) {materialFunction=&Swizzle}.YX, False=@n35) {defaultValue=true, parameterName=UseSwizzle}) {name=Distortion, nodeColor=<rgba 1 0.5156 0 1>}
  n36 = Divide(A=@n11.Default Range) {constB=0.05}
  NamedRerouteDeclaration(None=Multiply(A=@n9.G, B=Saturate(None=@n36))) {name=Opacity, nodeColor=<rgba 0 1 0.1328 1>}
  NamedRerouteDeclaration(None=@n1.A) {name=Alpha, nodeColor=<rgba 0.8047 0 1 1>}
  n37 = TextureCoordinate()
  n38 = Add(A=OneMinus(None=ComponentMask(None=@n37) {g=true}), B=NamedRerouteUsage()) {constB=-0.5}
  NamedRerouteDeclaration(None=@n38) {name=Reveal Mask, nodeColor=<rgba 1 0 0.8203 1>}
  NamedRerouteDeclaration(None=@n29.Reveal) {name=Reveal, nodeColor=<rgba 0.1016 1 0 1>}
  NamedRerouteDeclaration(None=@n29.WPO) {name=WPO, nodeColor=<rgba 1 0 0.2812 1>}
  OneMinus(None=@n38)
  VectorParameter() {parameterName=InsideColor, defaultValue=<rgba 0 0 0 1>}
 ]
}
