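# Niagara wind effect for the StudyBench emote, combining mirrored inner and outer wing meshes, a jet mesh, speed-line sprites and socket-sampled ribbon contrails.
# A summed absolute owner-velocity threshold switches spawning, with a System Preview value lifting the idle rates of all but the contrails.
# GlowCard follows the skeletal mesh with a mesh and airglow sprite tinted by time of day.
# System-level smoothed values modulate alpha, material parameters and glow offset.
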NiagaraSystem /BRCosmetics/Animation/Game/MainPlayer/Emotes/StudyBench/FX/NS_Emote_StudyBench_Wind
dependencies {
 usedDataInterfaces [
  &NiagaraDataInterfaceArrayFloat
  &NiagaraDataInterfaceCurve
  &NiagaraDataInterfaceArrayFloat3
  &NiagaraDataInterfaceVector2DCurve
  &NiagaraDataInterfaceColorCurve
 ]
 usedDynamicInputs [
  &SimulationPosition
  &Vector2DFromFloat
  &UniformRangedFloat
  &MakeFloatFromVector2d
  &RemapRange
 ]
 usedModules [
  &self:SystemSpawnScript.SystemScriptSource.SystemScriptGraph.NiagaraNodeAssignment_0.SetVariables_#1
  &SystemState
  &self:SystemSpawnScript.SystemScriptSource.SystemScriptGraph.NiagaraNodeAssignment_1.SetVariables_#2
  &InterpolateOverTime
  &self:SystemSpawnScript.SystemScriptSource.SystemScriptGraph.NiagaraNodeAssignment_2.SetVariables_#3
  &self:SystemSpawnScript.SystemScriptSource.SystemScriptGraph.NiagaraNodeAssignment_3.SetVariables_#4
  &self:SystemSpawnScript.SystemScriptSource.SystemScriptGraph.NiagaraNodeAssignment_4.SetVariables_#5
  &self:SystemSpawnScript.SystemScriptSource.SystemScriptGraph.NiagaraNodeAssignment_5.SetVariables_#6
  &self:SystemSpawnScript.SystemScriptSource.SystemScriptGraph.NiagaraNodeAssignment_6.SetVariables_#7
 ]
}
emitters {
 ContrailsOpt {
  emitterUpdateScript [
   &EmitterState(Life Cycle Mode=::NewEnumerator1, Loop Delay=0.07, Loop Duration=0.2, UseLoopDelay=true)
   &SpawnRate(
    SpawnRate=&Lerp_Float(
     A=3
     B=256
     Alpha=&MakeCustomFloatFromBool(
      Bool=&SetBoolByFloatComparison(
       A=&Add_Float(
        A=&Add_Float(
         A=&Abs_Float(Float:NiagaraFloat=&MakeFloatFromVector(Vector=$Engine.Owner.Velocity, Channel=::NewEnumerator0))
         B=&Abs_Float(Float:NiagaraFloat=&MakeFloatFromVector(Vector=$Engine.Owner.Velocity, Channel=::NewEnumerator2))
        )
        B=&Abs_Float(Float:NiagaraFloat=&MakeFloatFromVector(Vector=$Engine.Owner.Velocity, Channel=::NewEnumerator1))
       )
       B=200
      )
      True Float=&ClampFloat(Float=1)
      False Float=&ClampFloat(Float=0)
     )
    )
    Use Spawn Probability=false
   ) {module=SpawnRate001}
  ]
  particleSpawnScript [
   &FN_SetSpawnGroup()
   &InitializeRibbon(
    Color=$System.Contrails
    Lifetime=&SelectFloatFromArray(
     Direct Array Index=&SpawnGroup()
     Float Selection Array=DataInterface() {propertyValues {FloatData [0.45,0.45,0.15,0.15]}}
    )
    Ribbon Facing Mode:ENiagara_UnsetDirectSet=::NewEnumerator1
    Ribbon Facing Vector:Vector3f=<1 0 0>
    Ribbon Width=&SelectFloatFromArray(
     Direct Array Index=&SpawnGroup()
     Float Selection Array=DataInterface() {propertyValues {FloatData [35,35,10,10]}}
    )
    UsePositionOffset=false
    Write Material Random=true
   )
   &SkeletalMeshLocation(
    Bone / Socket Index=$Particles.SpawnGroup
    MaskSpawnGroup=false
    Mesh Sampling Type=::NewEnumerator5
    Skeletal Mesh=$System.Skeletal Mesh
    Socket Sampling Mode=::NewEnumerator5
   )
  ]
  particleUpdateScript [
   &ParticleState()
   &ScaleColor(
    Scale Alpha=&Multiply_Float(
     A=&FloatFromCurve(
      FloatCurve=DataInterface() {propertyValues {Curve {keys [
          <at 0 0 RCIM_Cubic>
          {interpMode=RCIM_Cubic, tangentMode=RCTM_User, time=0.2, value=1}
          {interpMode=RCIM_Cubic, tangentMode=RCTM_User, time=1, value=0}
         ]}}}
      Scale Curve=&FNTimeOfDayScalar(Night Scale=0.2, Morning Scale=0.2, Evening Scale=0.2)
     )
     B=1
    )
    Scale RGB=&Multiply_VectorByFloat(
     Vector=&MakeVectorFromLinearColorRGB(Color=&ColorFromCurve(ColorCurve=DataInterface() {propertyValues {
         AlphaCurve {keys []}
         BlueCurve {keys [<at 0 4.8655 RCIM_Cubic>,<at 0.4581 4.8655>,<at 0.8818 6.0417>,<at 0.9992 2.9167 RCIM_Cubic>]}
         GreenCurve {keys [<at 0 10 RCIM_Cubic>,<at 0.4581 10>,<at 0.8818 10>,<at 0.9992 5 RCIM_Cubic>]}
         RedCurve {keys [<at 0 1.1979 RCIM_Cubic>,<at 0.4581 1.1979>,<at 0.8818 0.5>,<at 0.9992 0 RCIM_Cubic>]}
        }}))
     Float=&FloatFromCurve(
      FloatCurve=DataInterface() {propertyValues {Curve {keys [<at 0 1>,<at 0.1 1.2>,<at 0.3 1>,<at 0.5 0.5>,<at 0.7 1.5>,<at 1 1>]}}}
      Scale Curve=0.8
     )
    )
    ScaleA=true
    ScaleRGB=false
   )
   &SolveForcesAndVelocity(Clamp Velocity=false, Limit Acceleration=false)
   &ScaleColor(Scale Alpha=$System.InterpolateOverTime001.InterpolateOverTime_Float.Moving Average, ScaleA=true, ScaleRGB=true) {module=ScaleColor001}
  ]
  rendererClasses [&NiagaraRibbonRendererProperties]
  renderers [{rendererClass=&NiagaraRibbonRendererProperties}]
 }
 GlowCard {
  emitterUpdateScript [
   &EmitterState(
    Inactive Response=::NewEnumerator1
    Life Cycle Mode=::NewEnumerator1
    Loop Behavior=::NewEnumerator1
    Loop Duration Mode=::NewEnumerator1
    UseLoopDelay=false
   )
   &SpawnBurst_Instantaneous(Use Spawn Probability=false, UseLoopCountLimit=false)
   &SpawnBurst_Instantaneous(Use Spawn Probability=false, UseLoopCountLimit=false) {module=SpawnBurst_Instantaneous001}
  ]
  particleSpawnScript [
   &InitializeParticle(
    Color=&FNTimeOfDayColor(
     Night Color=<rgba 0.9414 0.9511 1.4 1>
     Morning Color=<rgba 1.2419 1.2657 1.5 1>
     Day Color=<rgba 0.8818 0.9545 1.4 1>
     Evening Color=<rgba 0.7199 1.0063 1.5 1>
    )
    Mesh Scale Mode=::NewEnumerator3
    Mesh Uniform Scale=0.68
    Sprite Size Mode=::NewEnumerator3
    UsePositionOffset=false
    Write Material Random=true
   )
   &CameraOffset(Camera Offset Amount=1)
  ]
  particleUpdateScript [
   &ParticleState(Kill Particles When Lifetime Has Elapsed=false)
   &ScaleMeshSize(Scale Factor=&VectorFromFloat(Value=&FloatFromCurve(
      FloatCurve=DataInterface() {propertyValues {Curve {keys [{tangentMode=RCTM_User, time=0.5, value=0}, {tangentMode=RCTM_User, time=0.8, value=1}]}}}
      CurveIndex=$Emitter.Age
     )))
   &SkeletalMeshLocation(
    Bone Sampling Mode=::NewEnumerator4
    MaskSpawnGroup=false
    Sampled Position Offset=&MakeVector(
     X=0
     Y=&Lerp_Float(A=-8.2, B=-7, Alpha=$System.InterpolateOverTime002.InterpolateOverTime_Float.Moving Average)
     Z=0
    )
    Skeletal Mesh=$System.Skeletal Mesh
   )
   &SpriteFacingAndAlignment(
    Sprite Facing=&RotateVectorByQuaternion(Vector To Rotate=<0 -1 0>, Quaternion=$Particles.SkeletalMeshLocation.SampledRotation)
    WriteAlignment=false
    WriteFacing=true
   )
   &FN_Light_Attributes(
    Color=$Particles.Color
    Create Color Attribute=true
    Create Exponent Attribute=true
    Create Position Attribute=true
    Create Radius Attribute=true
    Exponent=1
    Position=&AddVectorToPosition(Position=$Particles.Position, Vector=<0 -10 0>)
    Radius=50
    Scale Color=10
   )
   /SetVariables(
    Particles.MeshColor:LinearColor=&Multiply_LinearColorByFloat(
     Linear Color=&FNTimeOfDayColor(
      Night Color=<rgba 0.6646 0.6727 1.3 1>
      Morning Color=<rgba 0.9652 1.1563 2 1>
      Day Color=<rgba 0.9165 1.0523 1.5 1>
      Evening Color=<rgba 0.6999 1.0945 1.5 1>
     )
     Float=&Multiply_Float(A=1.3, B=&FN_Select_Float_from_Time_Of_Day(Morning=1, Day=1, Evening=1, Night=0.8))
    )
    Particles.MeshPosition:NiagaraPosition=&AddVectorToPosition(
     Position=$Particles.Position
     Vector=&RotateVectorByQuaternion(
      Vector To Rotate=&MakeVector(X=0, Y=2.6, Z=0)
      Quaternion=$Particles.SkeletalMeshLocation.SampledRotation
     )
    )
   ) {isSetParameters=true}
   &ScaleColor(
    Scale Alpha=&Multiply_Float(A=1, B=&FN_Select_Float_from_Time_Of_Day(Morning=1, Day=1, Evening=1, Night=0.8))
    ScaleA=true
    ScaleRGB=true
   )
   /SetVariables(
    Particles.AirglowColor:LinearColor=&Multiply_LinearColorByFloat(Linear Color=$Particles.Color, Float=0.3)
    Particles.AirglowPosition:NiagaraPosition=&AddVectorToPosition(
     Position=$Particles.Position
     Vector=&RotateVectorByQuaternion(Vector To Rotate=<0 0 0>, Quaternion=$Particles.SkeletalMeshLocation.SampledRotation)
    )
    Particles.AirglowSize:Vector2f=&Multiply_Vector2D_ByFloat(
     Float=&FloatFromCurve(
      FloatCurve=DataInterface() {propertyValues {Curve {keys [{tangentMode=RCTM_User, time=0.4, value=0}, {tangentMode=RCTM_User, time=1, value=1}]}}}
      CurveIndex=$Emitter.Age
      Scale Curve=1.8
     )
     Vector2D=$Particles.SpriteSize
    )
   ) {isSetParameters=true}
   &CameraOffset(Camera Offset Amount=2) {module=CameraOffset001}
  ]
  rendererClasses [&NiagaraMeshRendererProperties, &NiagaraSpriteRendererProperties]
  renderers [
   {rendererClass=&NiagaraMeshRendererProperties, rendererIndex=0}
   {rendererClass=&NiagaraSpriteRendererProperties, rendererIndex=1}
  ]
 }
 Inner_Wings {
  emitterSpawnScript [/SetVariables(Emitter.Static bool=true) {isSetParameters=true, module=SetVariables_#8}]
  emitterUpdateScript [
   &EmitterState(Inactive Response=::NewEnumerator1, Life Cycle Mode=::NewEnumerator1, Loop Delay=1.5, Loop Duration=0.6, UseLoopDelay=false)
   &SpawnBurst_Instantaneous(
    Spawn Count=&MakeIntFromFloat(Float=&Lerp_Float(
      A=&Lerp_Float(A=0, B=2, Alpha=$System.Preview)
      B=4
      Alpha=&MakeCustomFloatFromBool(
       Bool=&SetBoolByFloatComparison(
        A=&Add_Float(
         A=&Add_Float(
          A=&Abs_Float(Float:NiagaraFloat=&MakeFloatFromVector(Vector=$Engine.Owner.Velocity, Channel=::NewEnumerator0))
          B=&Abs_Float(Float:NiagaraFloat=&MakeFloatFromVector(Vector=$Engine.Owner.Velocity, Channel=::NewEnumerator2))
         )
         B=&Abs_Float(Float:NiagaraFloat=&MakeFloatFromVector(Vector=$Engine.Owner.Velocity, Channel=::NewEnumerator1))
        )
        B=200
       )
       True Float=&ClampFloat(Float=1)
       False Float=&ClampFloat(Float=0)
      )
     ))
    Use Spawn Probability=false
    UseLoopCountLimit=false
   )
  ]
  particleSpawnScript [
   &FN_SetSpawnGroup(Number of Spawn Groups=2)
   /SetVariables(Particles.SizeDeterminatio:NiagaraFloat=&RandomRangeFloat(Minimum=0, Maximum=1)) {isSetParameters=true, module=SetVariables_#9}
   &InitializeParticle(
    Color=$System.Inner Color
    Lifetime=&RandomRangeFloat(Minimum=1.2, Maximum=1.5)
    Mass Mode=::NewEnumerator1
    Mesh Scale=&MakeVector(X=1, Y=&Lerp_Float(A=1.15, B=1.2, Alpha=$Particles.SizeDeterminatio), Z=1)
    Mesh Scale Mode=::NewEnumerator4
    Position Offset=&MakeVector(
     X=0
     Y=&Lerp_Float(A=25, B=25, Alpha=$Particles.SizeDeterminatio)
     Z=&SelectFloatFromArray(Direct Array Index=$Particles.SpawnGroup, Float Selection Array=DataInterface() {propertyValues {FloatData [3.5,3.5,0,0]}})
    )
    UsePositionOffset=true
    Write Material Random=true
   ) {module=InitializeParticle001}
   /SetVariables(Particles.SpriteFacing=<0 0 1>) {isSetParameters=true, module=SetVariables_#10}
  ]
  particleUpdateScript [
   &ParticleState()
   &ScaleMeshSize(Scale Factor=&MakeVector(
     X=1
     Y=&FloatFromCurve(FloatCurve=DataInterface() {propertyValues {Curve {keys [{tangentMode=RCTM_User, time=0, value=0.5}, {tangentMode=RCTM_User, time=0.5, value=1}]}}})
     Z=1
    )) {module=ScaleMeshSize002}
   &ScaleMeshSize(Scale Factor=&SelectVectorFromArray(
     Direct Array Index=$Particles.SpawnGroup
     Vector Selection Array=DataInterface() {propertyValues {FloatData [<1 1 1>, <-1 1 1>, <1 0.6 -1>, <-1 0.6 -1>]}}
    ))
   &ScaleMeshSize(Scale Factor=&MakeVector(
     X=&Lerp_Float(A=1.2, B=0.85, Alpha=$Particles.SizeDeterminatio)
     Y=&Lerp_Float(A=1.35, B=1.8, Alpha=$Particles.SizeDeterminatio)
     Z=&RandomRangeFloat(Minimum=1, Maximum=1)
    )) {module=ScaleMeshSize001}
   &SolveForcesAndVelocity(Clamp Velocity=false, Limit Acceleration=false)
   &ScaleColor(
    Scale Alpha=&FloatFromCurve(FloatCurve=DataInterface() {propertyValues {Curve {keys [
         {interpMode=RCIM_Cubic, tangentMode=RCTM_User, time=0, value=0}
         {interpMode=RCIM_Cubic, tangentMode=RCTM_User, time=0.5, value=1}
         {tangentMode=RCTM_User, time=1, value=0}
        ]}}})
    ScaleA=true
    ScaleRGB=true
   )
   &DynamicMaterialParameters(
    Index 0 Param 1=1
    Index 0 Param 2=&RandomRangeFloat(Minimum=0, Maximum=1)
    Index 0 Param 3=&FloatFromCurve(
     FloatCurve=DataInterface() {propertyValues {Curve {keys [{tangentMode=RCTM_User, time=0, value=0.2}, {tangentMode=RCTM_User, time=0.6, value=-1}]}}}
     Scale Curve=&Lerp_Float(A=$System.InterpolateOverTime.InterpolateOverTime_Float.Moving Average, B=1, Alpha=$System.Preview)
    )
    Index 0 Param 4=100
    Module.Index0Param0Write=true
    Module.Index0Param1Write=true
    Module.Index0Param2Write=true
    Module.Index0Param3Write=true
   ) {module=DynamicMaterialParameters001}
  ]
  rendererClasses [&NiagaraMeshRendererProperties]
  renderers [{rendererClass=&NiagaraMeshRendererProperties, rendererIndex=0}]
 }
 Jet {
  emitterUpdateScript [
   &EmitterState(Inactive Response=::NewEnumerator1, Life Cycle Mode=::NewEnumerator1, UseLoopDelay=false)
   &SpawnRate(
    SpawnRate=&Lerp_Float(
     A=&Lerp_Float(A=0, B=2, Alpha=$System.Preview)
     B=2
     Alpha=&MakeCustomFloatFromBool(
      Bool=&SetBoolByFloatComparison(
       A=&Add_Float(
        A=&Add_Float(
         A=&Abs_Float(Float:NiagaraFloat=&MakeFloatFromVector(Vector=$Engine.Owner.Velocity, Channel=::NewEnumerator0))
         B=&Abs_Float(Float:NiagaraFloat=&MakeFloatFromVector(Vector=$Engine.Owner.Velocity, Channel=::NewEnumerator2))
        )
        B=&Abs_Float(Float:NiagaraFloat=&MakeFloatFromVector(Vector=$Engine.Owner.Velocity, Channel=::NewEnumerator1))
       )
       B=200
      )
      True Float=&ClampFloat(Float=1)
      False Float=&ClampFloat(Float=0)
     )
    )
    Use Spawn Probability=false
   ) {module=SpawnRate001}
   /SetVariables(Emitter.float:NiagaraFloat=&Lerp_Float(A=0.9, B=0.6, Alpha=$System.InterpolateOverTime001.InterpolateOverTime_Float.Moving Average)) {isSetParameters=true}
  ]
  particleSpawnScript [
   &InitializeParticle(
    Color=<rgba 0.886 0.9009 1 1>
    Lifetime=0.5
    Mesh Scale=<0.6 0.6 3>
    Mesh Scale Mode=::NewEnumerator4
    Position Offset=<0 -100 2>
    UsePositionOffset=true
    Write Material Random=true
   )
   &InitialMeshOrientation(
    Mesh Orientation Mode=::NewEnumerator2
    Orientation Coordinate Space=::Local
    Orientation Vector=<0 0 1>
    Rotation=<0 0 -0.25>
    Rotation Coordinate Space=::NewEnumerator2
    Use Rotation Vector=true
   )
  ]
  particleUpdateScript [
   &ParticleState()
   &DynamicMaterialParameters(
    Index 0 Param 1=2
    Index 0 Param 2=-2
    Module.Index0Param0Write=true
    Module.Index0Param1Write=true
    Module.Index0Param2Write=true
    Module.Index0Param3Write=true
   )
   &ScaleColor(
    Scale Alpha=&FloatFromCurve(
     FloatCurve=DataInterface() {propertyValues {Curve {keys [
         {arriveTangent=9.0725, interpMode=RCIM_Cubic, leaveTangent=9.0725, tangentMode=RCTM_User, time=0, value=1}
         {arriveTangent=-2.5249, interpMode=RCIM_Cubic, leaveTangent=-2.5249, tangentMode=RCTM_User, time=0.1408, value=2}
         {arriveTangent=-0.0405, interpMode=RCIM_Cubic, leaveTangent=-0.0405, tangentMode=RCTM_User, time=1, value=1}
        ]}}}
     CurveIndex=$Emitter.float
    )
    Scale RGB=&MakeVectorFromLinearColorRGB(Color=&MultiplyLinearColor(Linear Color=<rgba 1 1 1 1>, Scale Factor=&MakeVector4_FromLinearColor(LinearColor=$System.ToD Color)))
    ScaleA=true
    ScaleRGB=true
   )
  ]
  rendererClasses [&NiagaraMeshRendererProperties]
  renderers [{rendererClass=&NiagaraMeshRendererProperties, rendererIndex=0}]
 }
 Outer_Wings {
  emitterSpawnScript [/SetVariables(Emitter.Static bool=true) {isSetParameters=true, module=SetVariables_#8}]
  emitterUpdateScript [
   &EmitterState(Inactive Response=::NewEnumerator1, Life Cycle Mode=::NewEnumerator1, Loop Delay=1.5, Loop Duration=0.7, UseLoopDelay=false)
   &SpawnRate(
    SpawnRate=&MakeFloatFromInt(Int=&MakeIntFromFloat(Float=&Lerp_Float(
       A=&Lerp_Float(A=0, B=&RandomRangeFloat(Minimum=2, Maximum=4), Alpha=$System.Preview)
       B=4
       Alpha=&MakeCustomFloatFromBool(
        Bool=&SetBoolByFloatComparison(
         A=&Add_Float(
          A=&Add_Float(
           A=&Abs_Float(Float:NiagaraFloat=&MakeFloatFromVector(Vector=$Engine.Owner.Velocity, Channel=::NewEnumerator0))
           B=&Abs_Float(Float:NiagaraFloat=&MakeFloatFromVector(Vector=$Engine.Owner.Velocity, Channel=::NewEnumerator2))
          )
          B=&Abs_Float(Float:NiagaraFloat=&MakeFloatFromVector(Vector=$Engine.Owner.Velocity, Channel=::NewEnumerator1))
         )
         B=200
        )
        True Float=&ClampFloat(Float=1)
        False Float=&ClampFloat(Float=0)
       )
      )))
    Use Spawn Probability=false
   ) {module=SpawnRate003}
  ]
  particleSpawnScript [
   &FN_SetSpawnGroup(Number of Spawn Groups=2)
   /SetVariables(Particles.SizeDeterminatio:NiagaraFloat=&RandomRangeFloat(Minimum=0, Maximum=1)) {isSetParameters=true, module=SetVariables_#9}
   &InitializeParticle(
    Color=$System.Outer Color
    Lifetime=&RandomRangeFloat(Minimum=1.2, Maximum=1.5)
    Mass Mode=::NewEnumerator1
    Mesh Scale=&MakeVector(X=1, Y=&Lerp_Float(A=1.15, B=1.2, Alpha=$Particles.SizeDeterminatio), Z=1)
    Mesh Scale Mode=::NewEnumerator4
    Position Offset=&MakeVector(
     X=0
     Y=&Lerp_Float(A=25, B=25, Alpha=$Particles.SizeDeterminatio)
     Z=&SelectFloatFromArray(Direct Array Index=$Particles.SpawnGroup, Float Selection Array=DataInterface() {propertyValues {FloatData [3.5,3.5,0,0]}})
    )
    UsePositionOffset=true
    Write Material Random=true
   ) {module=InitializeParticle001}
   /SetVariables(Particles.SpriteFacing=<0 0 1>) {isSetParameters=true, module=SetVariables_#10}
  ]
  particleUpdateScript [
   &ParticleState()
   &ScaleMeshSize(Scale Factor=&MakeVector(
     X=1
     Y=&FloatFromCurve(FloatCurve=DataInterface() {propertyValues {Curve {keys [{tangentMode=RCTM_User, time=0, value=0.5}, {tangentMode=RCTM_User, time=0.5, value=1}]}}})
     Z=1
    )) {module=ScaleMeshSize002}
   &ScaleMeshSize(Scale Factor=&SelectVectorFromArray(
     Direct Array Index=$Particles.SpawnGroup
     Vector Selection Array=DataInterface() {propertyValues {FloatData [<1 1 1>, <-1 1 1>, <1 0.6 -1>, <-1 0.6 -1>]}}
    ))
   &ScaleMeshSize(Scale Factor=&MakeVector(
     X=&Lerp_Float(A=1.1, B=&MaskFloatBySpawnGroup(Masked Float=1, Passthrough Float=0.7), Alpha=$Particles.SizeDeterminatio)
     Y=&Lerp_Float(A=1, B=1.5, Alpha=$Particles.SizeDeterminatio)
     Z=&RandomRangeFloat(Minimum=1, Maximum=1)
    )) {module=ScaleMeshSize001}
   &SolveForcesAndVelocity(Clamp Velocity=false, Limit Acceleration=false)
   &ScaleColor(
    Scale Alpha=&FloatFromCurve(FloatCurve=DataInterface() {propertyValues {Curve {keys [
         {interpMode=RCIM_Cubic, tangentMode=RCTM_User, time=0, value=0}
         {interpMode=RCIM_Cubic, tangentMode=RCTM_User, time=0.5, value=1}
         {tangentMode=RCTM_User, time=1, value=0}
        ]}}})
    ScaleA=true
    ScaleRGB=true
   )
   &DynamicMaterialParameters(
    Index 0 Param 1=1
    Index 0 Param 2=&RandomRangeFloat(Minimum=0, Maximum=1)
    Index 0 Param 3=&FloatFromCurve(
     FloatCurve=DataInterface() {propertyValues {Curve {keys [{tangentMode=RCTM_User, time=0, value=0.2}, {tangentMode=RCTM_User, time=0.6, value=-1}]}}}
     Scale Curve=&Lerp_Float(A=$System.InterpolateOverTime.InterpolateOverTime_Float.Moving Average, B=1, Alpha=$System.Preview)
    )
    Index 0 Param 4=100
    Module.Index0Param0Write=true
    Module.Index0Param1Write=true
    Module.Index0Param2Write=true
    Module.Index0Param3Write=true
   ) {module=DynamicMaterialParameters001}
  ]
  rendererClasses [&NiagaraMeshRendererProperties]
  renderers [{rendererClass=&NiagaraMeshRendererProperties, rendererIndex=0}]
 }
 Speed_Lines {
  emitterUpdateScript [
   &EmitterState(Life Cycle Mode=::NewEnumerator1, Loop Duration=0.95, UseLoopDelay=false)
   &SpawnRate(
    SpawnRate=&Lerp_Float(
     A=&Lerp_Float(A=0, B=20, Alpha=$System.Preview)
     B=20
     Alpha=&MakeCustomFloatFromBool(
      Bool=&SetBoolByFloatComparison(
       A=&Add_Float(
        A=&Add_Float(
         A=&Abs_Float(Float:NiagaraFloat=&MakeFloatFromVector(Vector=$Engine.Owner.Velocity, Channel=::NewEnumerator0))
         B=&Abs_Float(Float:NiagaraFloat=&MakeFloatFromVector(Vector=$Engine.Owner.Velocity, Channel=::NewEnumerator2))
        )
        B=&Abs_Float(Float:NiagaraFloat=&MakeFloatFromVector(Vector=$Engine.Owner.Velocity, Channel=::NewEnumerator1))
       )
       B=200
      )
      True Float=&ClampFloat(Float=1)
      False Float=&ClampFloat(Float=0)
     )
    )
    Use Spawn Probability=false
   ) {module=SpawnRate001}
  ]
  particleSpawnScript [
   &InitializeParticle(
    Color=&MultiplyLinearColor(Linear Color=$System.Contrails, Scale Factor=&MakeVector4(X=1, Y=1, Z=1, W=1))
    Lifetime Max=0.2
    Lifetime Min=0.15
    Lifetime Mode=::NewEnumerator1
    Position Mode=::NewEnumerator1
    Sprite Size Max=<80 150>
    Sprite Size Min=<20 50>
    Sprite Size Mode=::NewEnumerator2
    UsePositionOffset=false
    Write Material Random=true
   )
   &ShapeLocation(
    Box Size=<100 100 200>
    MaskSpawnGroup=false
    Offset=<0 50 0>
    Offset Coordinate Space=::Local
    Offset Mode=::NewEnumerator0
    Override Seed=false
    Shape Primitive=::NewEnumerator2
    Surface Only=false
    UseOwnerScale=true
   )
   &AddVelocity(Velocity=<0 -1000 0>) {module=AddVelocity001}
   &CameraOffset(Camera Offset Amount=&FN_FrontEndFloatModifier(Gameplay=0, Front End=-30))
  ]
  particleUpdateScript [
   &ParticleState()
   &Drag(Drag=5, Use Linear Drag=true, Use Rotational Drag=false)
   &SolveForcesAndVelocity(Clamp Velocity=false, Limit Acceleration=false)
   &ScaleColor(
    Scale Alpha=&FloatFromCurve(FloatCurve=DataInterface() {propertyValues {Curve {keys [
         {interpMode=RCIM_Cubic, tangentMode=RCTM_User, time=0.1552, value=0}
         {interpMode=RCIM_Cubic, tangentMode=RCTM_User, time=0.5, value=1}
         {tangentMode=RCTM_User, time=1, value=0}
        ]}}})
    ScaleA=true
    ScaleRGB=true
   )
   &ScaleSpriteSize(
    Non-Uniform Curve Sprite Scale=DataInterface() {propertyValues {
      XCurve {keys [{tangentMode=RCTM_User, time=0, value=0.5}, {tangentMode=RCTM_User, time=1, value=1.5}]}
      YCurve {keys [{tangentMode=RCTM_User, time=0, value=1}, {tangentMode=RCTM_User, time=1, value=0.5}]}
     }}
    Scale Sprite Size Mode=::NewEnumerator3
   )
   /SetVariables(Particles.New Position:NiagaraPosition=&Add_Vector(
     A=Linked() {linkedVariable {name=Particles.Position, type=NiagaraPosition}}
     B=Linked() {linkedVariable {name=Engine.Emitter.SimulationPosition, type=NiagaraPosition}}
    )) {isSetParameters=true}
  ]
  rendererClasses [&NiagaraSpriteRendererProperties]
  renderers [{rendererClass=&NiagaraSpriteRendererProperties}]
 }
}
summary {
 systemName=NS_Emote_StudyBench_Wind
 emitters [
  {emitterName=Outer_Wings, rendererClasses [&NiagaraMeshRendererProperties]}
  {emitterName=Jet, rendererClasses [&NiagaraMeshRendererProperties]}
  {emitterName=Speed_Lines, rendererClasses [&NiagaraSpriteRendererProperties]}
  {emitterName=Inner_Wings, rendererClasses [&NiagaraMeshRendererProperties]}
  {emitterName=ContrailsOpt, rendererClasses [&NiagaraRibbonRendererProperties]}
  {emitterName=GlowCard, rendererClasses [&NiagaraMeshRendererProperties, &NiagaraSpriteRendererProperties]}
 ]
 userVariables [
  User.bIsFrontEnd {type=NiagaraBool}
  User.bIsFrontEndPreview {type=NiagaraBool}
  User.SkeletalMesh {
   type=NiagaraDataInterfaceSkeletalMesh
   defaultValue {
    struct=NiagaraExt_VariableValue_DataInterface
    value {dataInterface=&self:NiagaraDataInterfaceSkeletalMesh_1}
   }
  }
 ]
}
