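# Niagara system for the StudyBench emote that kicks up wind puffs beneath the character as it moves.
# One sprite emitter samples the skeletal mesh, offsets puffs downward, launches them back and up with drag, and animates flipbook, size and fade, tinted by time of day.
# Above an owner velocity threshold, a time-interpolated system value drives spawn rate and also scales alpha; a Preview value allows spawning below it.
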NiagaraSystem /BRCosmetics/Animation/Game/MainPlayer/Emotes/StudyBench/FX/NS_Emote_StudyBench_WindKickup
dependencies {
 usedDataInterfaces [
  &NiagaraDataInterfaceCurve
  &NiagaraDataInterfaceSpriteRendererInfo
  &NiagaraDataInterfaceSkeletalMesh
 ]
 usedDynamicInputs [&RemapRange]
 usedModules [
  &self:SystemSpawnScript.SystemScriptSource.SystemScriptGraph.NiagaraNodeAssignment_2.SetVariables_#1
  &SystemState
  &self:SystemSpawnScript.SystemScriptSource.SystemScriptGraph.NiagaraNodeAssignment_0.SetVariables_#2
  &InterpolateOverTime
  &self:SystemSpawnScript.SystemScriptSource.SystemScriptGraph.NiagaraNodeAssignment_1.SetVariables_#3
 ]
}
emitters {Em_Wind_Kickup {
  emitterUpdateScript [
   &EmitterState(Inactive Response=::NewEnumerator1, Life Cycle Mode=::NewEnumerator1, UseLoopDelay=false)
   &SpawnRate(
    SpawnRate=&Lerp_Float(
     A=&Lerp_Float(A=0, B=24, Alpha=$System.Preview)
     B=&Lerp_Float(A=0, B=24, Alpha=$System.InterpolateOverTime001.InterpolateOverTime_Float.Moving Average)
     Alpha=&MakeCustomFloatFromBool(
      Bool=&SetBoolByFloatComparison(
       A=&Add_Float(
        A=&Add_Float(
         A=&Abs_Float(Float:NiagaraFloat=&MakeFloatFromVector(Vector=$Engine.Owner.Velocity, Channel=::NewEnumerator0))
         B=&Abs_Float(Float:NiagaraFloat=&MakeFloatFromVector(Vector=$Engine.Owner.Velocity, Channel=::NewEnumerator2))
        )
        B=&Abs_Float(Float:NiagaraFloat=&MakeFloatFromVector(Vector=$Engine.Owner.Velocity, Channel=::NewEnumerator1))
       )
       B=200
      )
      True Float=&ClampFloat(Float=1)
      False Float=&ClampFloat(Float=0)
     )
    )
    Use Spawn Probability=false
   ) {module=SpawnRate001}
  ]
  particleSpawnScript [
   &SkeletalMeshLocation(Bone Sampling Mode=::NewEnumerator4, MaskSpawnGroup=false, Position Sampling=::NewEnumerator1, Skeletal Mesh=$System.Skeletal Mesh)
   &InitializeParticle(
    Color=&FNTimeOfDayColor(
     Night Color=<rgba 0.5804 0.672 0.131 0.2>
     Morning Color=<rgba 0.9373 0.8392 0.7451 0.2>
     Day Color=<rgba 0.5804 0.8235 1 0.2>
     Evening Color=<rgba 0.9373 0.8392 0.7451 0.2>
    )
    Lifetime=&RandomRangeFloat(Minimum=0.5, Maximum=0.65)
    Position=$Output.SkeletalMeshLocation.SampledPositionWithoutOffset
    Position Mode=::NewEnumerator1
    Position Offset=<0 0 -35>
    Sprite Rotation Mode=::NewEnumerator2
    Sprite Size Mode=::NewEnumerator1
    Sprite UV Mode=::NewEnumerator4
    Sprite UV Scale:Vector2f=<-1 1>
    Uniform Sprite Size Max=125
    Uniform Sprite Size Min=75
    UsePositionOffset=true
    Write Material Random=true
   )
   &AddVelocity(Velocity=<0 -200 100>)
   &ApplyInitialForces()
   &ShapeLocation(
    Cylinder Radius=25
    Enable Lathe Profile=true
    MaskSpawnGroup=false
    Override Seed=false
    Rotation Mode=::NewEnumerator1
    Shape Primitive=::NewEnumerator1
    Surface Only=false
    UseOwnerScale=true
   ) {enabled=false}
  ]
  particleUpdateScript [
   &ParticleState()
   &SubUVAnimation(
    Play Rate=0.5
    Playback Scale Curve=DataInterface() {propertyValues {Curve {keys [{tangentMode=RCTM_User, time=0, value=0}, {tangentMode=RCTM_User, time=1, value=1}]}}}
    "Sprite Renderer "=DataInterface() {propertyValues {SpriteRenderer=&self:Em_Wind_Kickup_0.NiagaraSpriteRendererProperties_0}}
    SubUV Animation Mode:ENiagara_SubUVLookupModeV2=::NewEnumerator3
    UseEndFrame=false
    UsePlayRateScale=false
    UseStartFrame=false
   )
   &ScaleSpriteSize(Uniform Curve Sprite Scale=DataInterface() {propertyValues {Curve {keys [<at 0 0.5 RCIM_Cubic>, {arriveTangent=2.913, tangentMode=RCTM_User, time=1, value=2}]}}})
   &Drag(Use Linear Drag=true, Use Rotational Drag=true)
   &SolveForcesAndVelocity(Clamp Velocity=false, Limit Acceleration=false)
   &ScaleColor(
    Scale Alpha=&FloatFromCurve(
     FloatCurve=DataInterface() {propertyValues {Curve {keys [
         {arriveTangent=-2000.0001, interpMode=RCIM_Cubic, time=0, value=0}
         <at 0 0>
         {interpMode=RCIM_Cubic, tangentMode=RCTM_User, time=0.0759, value=0.2}
         {tangentMode=RCTM_User, time=1, value=0}
        ]}}}
     Scale Curve=&Lerp_Float(A=$System.InterpolateOverTime001.InterpolateOverTime_Float.Moving Average, B=1, Alpha=$System.Preview)
    )
    Scale RGB=&MakeVectorFromLinearColorRGB(Color=&MultiplyLinearColor(Linear Color=<rgba 1 1 1 1>, Scale Factor=&MakeVector4_FromLinearColor(LinearColor=<rgba 1 1 1 1>)))
    ScaleA=true
    ScaleRGB=true
   )
  ]
  rendererClasses [&NiagaraSpriteRendererProperties]
  renderers [{rendererClass=&NiagaraSpriteRendererProperties}]
 }}
summary {
 systemName=NS_Emote_StudyBench_WindKickup
 emitters [{emitterName=Em_Wind_Kickup, rendererClasses [&NiagaraSpriteRendererProperties]}]
 userVariables [
  User.bIsInWater {type=NiagaraBool}
  User.JetActiveCurve {type=NiagaraFloat, defaultValue=5, description=""}
 ]
}
